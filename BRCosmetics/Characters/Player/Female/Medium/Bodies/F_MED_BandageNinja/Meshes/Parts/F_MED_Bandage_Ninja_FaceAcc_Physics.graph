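# Physics asset for the Bandage Ninja female medium body's face accessory mesh, defining capsule bodies and joint limits for its ponytail and bangs.
# The ponytail base hangs off the head with a wide swing limit, and segments A through G each constrain directly to that base rather than chaining to one another.
# Left and right bangs form two-link chains from the head, with tighter twist limits than the ponytail.
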PhysicsAsset /BRCosmetics/Characters/Player/Female/Medium/Bodies/F_MED_BandageNinja/Meshes/Parts/F_MED_Bandage_Ninja_FaceAcc_Physics
bodies {
 dyn_bangs_01_l [{shapeType=Capsule, length=2, radius=2, center=<1 0 0>, rotation=<rot 90 0 0>}]
 dyn_bangs_01_r [{shapeType=Capsule, length=2, radius=2, center=<1 0 0>, rotation=<rot 90 0 0>}]
 dyn_bangs_02_l [{shapeType=Capsule, length=2, radius=2, center=<1 0 0>, rotation=<rot 90 0 0>}]
 dyn_bangs_02_r [{shapeType=Capsule, length=2, radius=2, center=<1 0 0>, rotation=<rot 90 0 0>}]
 head [{shapeType=Capsule, length=1, radius=1, center=<0 0.5135 0>}]
 ponytail_A [{shapeType=Capsule, length=2, radius=3, center=<0 0.1 0>, rotation=<rot 0 0 90>}]
 ponytail_B [{shapeType=Capsule, length=2, radius=3, center=<0 0.1 0>, rotation=<rot 0 0 90>}]
 ponytail_C [{shapeType=Capsule, length=2, radius=3, center=<2 0 0>, rotation=<rot 90 0 0>}]
 ponytail_D [{shapeType=Capsule, length=2, radius=3, center=<2 0 0>, rotation=<rot 90 0 0>}]
 ponytail_E [{shapeType=Capsule, length=2, radius=3, center=<2 0 0>, rotation=<rot 90 0 0>}]
 ponytail_F [{shapeType=Capsule, length=2, radius=3, center=<2 0 0>, rotation=<rot 90 0 0>}]
 ponytail_G [{shapeType=Capsule, length=2, radius=3, center=<2 0 0>, rotation=<rot 90 0 0>}]
 ponytail_base [{shapeType=Capsule, length=2.4666, radius=3.718, center=<1.6043 0 0>, rotation=<rot 90 0 0>}]
}
constraints [
 {bone1Name=ponytail_base, bone2Name=head, swing1LimitDegrees=75, swing2LimitDegrees=75, twistLimitDegrees=10}
 {bone1Name=ponytail_A, bone2Name=ponytail_base, swing1LimitDegrees=45, swing2LimitDegrees=45, twistLimitDegrees=10}
 {bone1Name=ponytail_B, bone2Name=ponytail_base, swing1LimitDegrees=45, swing2LimitDegrees=45, twistLimitDegrees=10}
 {bone1Name=ponytail_C, bone2Name=ponytail_base, swing1LimitDegrees=45, swing2LimitDegrees=45, twistLimitDegrees=10}
 {bone1Name=ponytail_D, bone2Name=ponytail_base, swing1LimitDegrees=45, swing2LimitDegrees=45, twistLimitDegrees=10}
 {bone1Name=ponytail_E, bone2Name=ponytail_base, swing1LimitDegrees=45, swing2LimitDegrees=45, twistLimitDegrees=10}
 {bone1Name=ponytail_F, bone2Name=ponytail_base, swing1LimitDegrees=45, swing2LimitDegrees=45, twistLimitDegrees=10}
 {bone1Name=ponytail_G, bone2Name=ponytail_base, swing1LimitDegrees=45, swing2LimitDegrees=45, twistLimitDegrees=10}
 {bone1Name=dyn_bangs_01_r, bone2Name=head, swing1LimitDegrees=45, swing2LimitDegrees=45, twistLimitDegrees=5}
 {bone1Name=dyn_bangs_02_r, bone2Name=dyn_bangs_01_r, swing1LimitDegrees=45, swing2LimitDegrees=45, twistLimitDegrees=5}
 {bone1Name=dyn_bangs_01_l, bone2Name=head, swing1LimitDegrees=45, swing2LimitDegrees=45, twistLimitDegrees=5}
 {bone1Name=dyn_bangs_02_l, bone2Name=dyn_bangs_01_l, swing1LimitDegrees=45, swing2LimitDegrees=45, twistLimitDegrees=5}
]
