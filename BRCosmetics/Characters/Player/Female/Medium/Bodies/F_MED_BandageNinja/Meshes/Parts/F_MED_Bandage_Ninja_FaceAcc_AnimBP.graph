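# Animation Blueprint for the female Bandage Ninja face accessory, adding secondary physics motion such as swaying bangs.
# It copies the pose from the attached parent mesh, then chains four root-relative Anim Dynamics nodes on the left and right bang bones, with reduced gravity and angular limits.
# A final RigidBody node simulates the accessory's physics asset relative to the head bone.
# The update event has no logic.
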AnimBlueprint /BRCosmetics/Characters/Player/Female/Medium/Bodies/F_MED_BandageNinja/Meshes/Parts/F_MED_Bandage_Ninja_FaceAcc_AnimBP
target_skeleton=&/F_MED_Bandage_Ninja_FaceAcc_Skeleton
graphs [
 AnimGraph {nodes [
   n1 = Anim Dynamics - Bone: dyn_bangs_01_l(ComponentPose=Local To Component(LocalPose=Copy Pose From Mesh() {bUseAttachedParent=true}), Alpha=1, ExternalForce="0,0,0") {
    angularDampingOverride=0.9999
    bGravityOverrideInSimSpace=true
    bOverrideAngularDamping=true
    gravityScale=0.35
    simulationSpace=RootRelative
    alphaBoolBlend {blendOption=Linear}
    boundBone {boneName=dyn_bangs_01_l}
    componentLinearAccScale {z=0.5}
    componentLinearVelScale {z=0.5}
    physicsBodyDefinitions [{
      sphereCollisionRadius=0
      boundBone {boneName=dyn_bangs_01_l}
      boxExtents=<5 20 20>
      constraintSetup {angularLimitsMax=<0 30 10>, angularLimitsMin=<0 0 -10>, angularTarget=<0 0 0>}
      localJointOffset=<5 0 0>
     }]
   }
   n2 = Anim Dynamics - Bone: dyn_bangs_02_l(ComponentPose=@n1, Alpha=1, ExternalForce="0,0,0") {
    angularDampingOverride=1
    bGravityOverrideInSimSpace=true
    bOverrideAngularDamping=true
    gravityScale=0.35
    simulationSpace=RootRelative
    alphaBoolBlend {blendOption=Linear}
    boundBone {boneName=dyn_bangs_02_l}
    componentLinearAccScale {z=0.5}
    componentLinearVelScale {z=0.5}
    physicsBodyDefinitions [{
      sphereCollisionRadius=0
      boundBone {boneName=dyn_bangs_02_l}
      boxExtents=<5 20 20>
      constraintSetup {angularLimitsMax=<0 30 10>, angularLimitsMin=<0 0 -10>, angularTarget=<0 0 0>}
      localJointOffset=<5 0 0>
     }]
   }
   n3 = Anim Dynamics - Bone: dyn_bangs_01_r(ComponentPose=@n2, Alpha=1, ExternalForce="0,0,0") {
    angularDampingOverride=0.9999
    bGravityOverrideInSimSpace=true
    bOverrideAngularDamping=true
    gravityScale=0.35
    simulationSpace=RootRelative
    alphaBoolBlend {blendOption=Linear}
    boundBone {boneName=dyn_bangs_01_r}
    componentLinearAccScale {z=0.5}
    componentLinearVelScale {z=0.5}
    physicsBodyDefinitions [{
      sphereCollisionRadius=0
      boundBone {boneName=dyn_bangs_01_r}
      boxExtents=<5 20 20>
      constraintSetup {angularLimitsMax=<0 30 10>, angularLimitsMin=<0 -15 -10>, angularTarget=<0 0 0>}
      localJointOffset=<5 0 0>
     }]
   }
   n4 = Anim Dynamics - Bone: dyn_bangs_02_r(ComponentPose=@n3, Alpha=1, ExternalForce="0,0,0") {
    angularDampingOverride=1
    bGravityOverrideInSimSpace=true
    bOverrideAngularDamping=true
    gravityScale=0.35
    simulationSpace=RootRelative
    alphaBoolBlend {blendOption=Linear}
    boundBone {boneName=dyn_bangs_02_r}
    componentLinearAccScale {z=0.5}
    componentLinearVelScale {z=0.5}
    physicsBodyDefinitions [{
      sphereCollisionRadius=0
      boundBone {boneName=dyn_bangs_02_r}
      boxExtents=<5 20 20>
      constraintSetup {angularLimitsMax=<0 30 10>, angularLimitsMin=<0 0 -10>, angularTarget=<0 0 0>}
      localJointOffset=<5 0 0>
     }]
   }
   n5 = RigidBody(ComponentPose=@n4) {
    bClampLinearTranslationLimitToRefPose=true
    overridePhysicsAsset=&/F_MED_Bandage_Ninja_FaceAcc_Physics
    simulationSpace=BaseBoneSpace
    alphaBoolBlend {blendOption=Linear}
    baseBoneRef {boneName=head}
    componentAppliedLinearAccClamp=<3000 3000 4200>
    componentLinearAccScale=<0.3 0.3 0.45>
    componentLinearVelScale=<0.15 0.15 0.225>
   }
   Output Pose(Result=Component To Local(ComponentPose=@n5)) {layerGroup=None}
  ]}
 EventGraph {nodes [Event BlueprintUpdateAnimation()]}
]
variables []
